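# M_GlitterEdge is an additive, unlit, two-sided Niagara sprite material that draws a glittering ring for the Buffet umbrella gadget FX.
# Custom atan2 polar UVs pan a caustics texture, with Cartesian derivatives driving mip selection, tinted by a panning rainbow sparkle LUT.
# Two subtracted sphere masks cut the ring, softened by depth fade and dimmed at grazing angles.
# Particle color and alpha scale the result before the shared emissive-adjustment function.
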Material /BRCosmetics/Gadgets/Assets/VinderTech_Umbrella/Umbrella_Buffet/FX/M_GlitterEdge
graph {
 properties {
  bUsedWithNiagaraSprites=true
  blendMode=BLEND_Additive
  pixelDepthOffsetMode=PDOM_Legacy
  shadingModel=MSM_Unlit
  twoSided=true
  usedShadingModels=Unlit
 }
 nodes [
  n1 = ParticleColor()
  n2 = Saturate(None=@n1.A)
  n3 = Add(A=Divide(A=TextureCoordinate()) {constB=0.25}) {constB=-2}
  MaterialExpressionComponentMask_20 = ComponentMask(None=@n3) {r=true}
  MaterialExpressionComponentMask_21 = ComponentMask(None=@n3) {g=true}
  n4 = Custom!(x=@MaterialExpressionComponentMask_20, y=@MaterialExpressionComponentMask_21) {
   code="return (atan2(y,x));"
   description=atan2
   outputType=CMOT_Float1
   inputs [
    {
     inputName=x
     input {expression=&self:MaterialExpressionComponentMask_20}
    }
    {
     inputName=y
     input {expression=&self:MaterialExpressionComponentMask_21}
    }
   ]
  }
  n5 = Frac(None=Divide(A=ComponentMask(None=@n4) {r=true}, B=Constant() {r=6.2832}))
  n6 = Multiply(A=@n3, B=@n3)
  n7 = SquareRoot(None=Add(A=ComponentMask(None=@n6) {r=true}, B=ComponentMask(None=@n6) {g=true}))
  n8 = TextureSample!(
   UVs=Panner(Coordinate=AppendVector(A=@n5, B=@n7)) {speedX=0.1, speedY=-0.4}
   Tex
   "DDX(UVs)"=DDX(Value=@n3)
   "DDY(UVs)"=DDY(Value=@n3)
   Apply View MipBias
  ) {mipValueMode=TMVM_Derivative, texture=&T_ShieldCaustics}
  n9 = Panner() {speedX=0.2, speedY=0.1}
  n10 = Desaturation(
   None=Multiply(
    A=Power(Base=@n8.R) {constExponent=0.5}
    B=TextureSample!(UVs=@n9, Tex, Apply View MipBias) {texture=/Game/Gadgets/Assets/VinderTech_Umbrella/Umbrella_Buffet/Materials/T_SequinDiamond_RainbowSparkle_LUT}.RGB
   )
   Fraction=Constant() {r=0.1}
  )
  n11 = SphereMask(A=TextureCoordinate(), B=Constant2Vector() {g=0.5, r=0.5}) {attenuationRadius=0.48, hardnessPercent=80}
  n12 = SphereMask(A=TextureCoordinate(), B=Constant2Vector() {g=0.5, r=0.5}) {attenuationRadius=0.38, hardnessPercent=80}
  n13 = Saturate(None=Subtract(A=Multiply(A=@n11, B=@n11), B=Multiply(A=@n12, B=@n12)))
  n14 = DepthFade(Opacity=Multiply(A=Multiply(A=@n10, B=@n13)) {constB=0.75}) {fadeDistanceDefault=32}
  n15 = Abs(None=DotProduct(A=CameraVectorWS(), B=VertexNormalWS()))
  n16 = Power(Base=Multiply(A=Saturate(None=@n15)) {constB=1.5}, Exp=Constant() {r=0.65})
  MaterialFunctionCall(Emissive=Multiply(A=Multiply(A=@n1.RGB, B=@n2), B=Multiply(A=@n14, B=@n16))) {materialFunction=&MF_FX_EmissiveAdjustments}
 ]
}
