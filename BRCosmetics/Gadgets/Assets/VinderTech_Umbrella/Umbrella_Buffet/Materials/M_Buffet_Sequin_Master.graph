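# Master material for the Buffet VinderTech umbrella glider cosmetic, adding a sequin look to skeletal meshes.
# MF_BuffetSequins generates the sequin layer with a constant full mask, while a plain layer sets base color, metallic and normal from textures.
# The inverted red channel of the M texture blends the two, so sequins show where that mask is white.
# The blended attributes pass through the shared MF_TechArtOverrides_accessories function.
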Material /BRCosmetics/Gadgets/Assets/VinderTech_Umbrella/Umbrella_Buffet/Materials/M_Buffet_Sequin_Master
graph {
 properties {
  bUseMaterialAttributes=true
  bUsedWithMeshDeformer=true
  bUsedWithSkeletalMesh=true
  pixelDepthOffsetMode=PDOM_Legacy
  usedShadingModels=Default Lit
 }
 nodes [
  n1 = Constant() {r=1}
  n2 = TextureSample() {texture=/BRCosmetics/Gadgets/Assets/VinderTech_Umbrella/Umbrella_Buffet/Textures/T_Umbrella_Buffet_Sequin_D}
  n3 = TextureSample() {samplerType=SAMPLERTYPE_Masks, texture=/BRCosmetics/Gadgets/Assets/VinderTech_Umbrella/Umbrella_Buffet/Textures/T_Umbrella_Buffet_S}
  n4 = TextureSample() {samplerType=SAMPLERTYPE_Normal, texture=/BRCosmetics/Gadgets/Assets/VinderTech_Umbrella/Umbrella_Buffet/Textures/T_Umbrella_Buffet_N}
  n5 = TextureSample() {samplerType=SAMPLERTYPE_LinearColor, texture=/BRCosmetics/Gadgets/Assets/VinderTech_Umbrella/Umbrella_Buffet/Textures/T_Umbrella_Buffet_M}
  MaterialFunctionCall(In=BlendMaterialAttributes(
    A=MaterialFunctionCall!(SequinColorInner, Fresnel, inMat, Mask=@n1, SequinColorOuter) {materialFunction=/Game/Gadgets/Assets/VinderTech_Umbrella/Umbrella_Buffet/Materials/MF_BuffetSequins}.Result
    B=SetMaterialAttributes!(MaterialAttributes, Base Color=@n2.RGB, Metallic=@n3.G, Normal=@n4.RGB) {attributeSetTypes [#1, #2, #3]}
    Alpha=OneMinus(None=@n5.R)
   )) {materialFunction=&MF_TechArtOverrides_accessories}
 ]
}
